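# Translucent deferred decal material for the AwfulDuck item's wave effect, producing an animated wave pattern inside a soft ring.
# Tiled radial coordinates supply one component that is offset by time times a pan speed, then fed through a sine to animate the bands.
# Subtracting a smaller UV sphere mask from a larger one forms the ring that masks the sine.
# A dark blue Color parameter passes through the shared FX emissive adjustments function.
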Material /Game/Athena/Items/AwfulDuck/FX/MD_AwfulDuck_WaveDecal
graph {
 properties {blendMode=BLEND_Translucent, materialDomain=MD_DeferredDecal, pixelDepthOffsetMode=PDOM_Legacy, usedShadingModels=Default Lit}
 nodes [
  Constant() {r=1}
  Constant()
  Constant() {r=0.5}
  MaterialFunctionCall(Emissive=VectorParameter() {parameterName=Color, defaultValue=<rgba 0.0004 0.0022 0.0052 1>}.RGB) {materialFunction=&MF_FX_EmissiveAdjustments}
  n1 = MaterialFunctionCall() {materialFunction=&VectorToRadialValue}
  n2 = ScalarParameter() {defaultValue=2, parameterName=Tile}
  n3 = Add(
   A=Multiply(A=Time(), B=ScalarParameter() {defaultValue=1.5, parameterName=PanSpeed})
   B=MaterialFunctionCall(Float2=Multiply(A=@n1.Radial Coordinates, B=@n2) {constB=2}) {materialFunction=&BreakOutFloat2Components}.G
  )
  n4 = Constant() {r=0.5}
  n5 = SphereMask(A=TextureCoordinate(), B=@n4, Radius=@n4) {attenuationRadius=0.5, hardnessPercent=0}
  n6 = SphereMask(A=TextureCoordinate(), B=Constant() {r=0.5}, Radius=Constant() {r=0.2}) {attenuationRadius=0.5, hardnessPercent=0}
  Multiply(A=Multiply(A=Sine(None=@n3), B=Subtract(A=@n5, B=@n6))) {constB=1.3}
  ParticleColor()
  TextureCoordinate()
 ]
}
