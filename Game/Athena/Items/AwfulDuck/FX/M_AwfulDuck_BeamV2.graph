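# Translucent, unlit, two-sided Niagara mesh-particle material for the AwfulDuck item's beam effect.
# Axis-based view-facing falloff, separate for inside and outside faces, combines with UV, near-camera and depth fades to shape opacity.
# Two panning noises distort scrolling curve-atlas colors, blended with a lengthwise gradient and optional moving rings.
# Depth-based intersection highlights, switched by quality and shading path, boost emissive color, which particle color can tint or replace.
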Material /Game/Athena/Items/AwfulDuck/FX/M_AwfulDuck_BeamV2
graph {
 properties {
  bUsedWithNiagaraMeshParticles=true
  blendMode=BLEND_Translucent
  pixelDepthOffsetMode=PDOM_Legacy
  shadingModel=MSM_Unlit
  twoSided=true
  usedShadingModels=Unlit
 }
 nodes [
  Constant3Vector() {constant=<rgba 1 0 0 0>}
  Constant3Vector() {constant=<rgba 1 0 0 0>}
  n1 = Subtract(A=WorldPosition().XYZ, B=CameraPositionWS())
  n2 = VectorParameter() {parameterName=MeshAxis, defaultValue=<rgba 1 0 0 0>}
  n3 = Transform(None=@n2.RGB) {transformSourceType=TRANSFORMSOURCE_Local}
  n4 = Normalize(VectorInput=CrossProduct(A=Normalize(VectorInput=@n1), B=@n3))
  n5 = Normalize(VectorInput=Multiply(A=CrossProduct(A=@n4, B=@n3), B=Constant3Vector() {constant=<rgba -1 -1 -1 0>}))
  n6 = VertexNormalWS()
  n7 = MaterialFunctionCall(Float2=TextureCoordinate()) {materialFunction=&BreakOutFloat2Components}
  n8 = Multiply(A=Power(Base=DotProduct(A=@n5, B=@n6), Exp=ScalarParameter() {defaultValue=2.571, parameterName=InsideHardness}), B=OneMinus(None=@n7.G))
  n9 = ScalarParameter() {defaultValue=1.4257, parameterName=Param_4}
  n10 = Transform(None=@n2.RGB) {transformSourceType=TRANSFORMSOURCE_Local}
  n11 = MaterialFunctionCall!(Plane Normal=@n10, Vector=Subtract(A=ObjectPositionWS(), B=CameraPositionWS())) {materialFunction=/Engine/Functions/Engine_MaterialFunctions02/Math/ProjectVectorOntoPlane}
  n12 = Normalize(VectorInput=CrossProduct(A=Normalize(VectorInput=@n11.Result), B=@n10))
  n13 = DotProduct(A=Normalize(VectorInput=CrossProduct(A=@n12, B=@n10)), B=VertexNormalWS())
  n14 = OneMinus(None=MaterialFunctionCall(Float2=TextureCoordinate()) {materialFunction=&BreakOutFloat2Components}.G)
  n15 = Multiply(A=Multiply(A=@n13, B=@n13), B=Multiply(A=@n14, B=@n14))
  n16 = ScalarParameter() {defaultValue=0.7298, parameterName=OutsideHardness}
  n17 = TwoSidedSign()
  n18 = Saturate(None=Multiply(A=@n7.G) {constB=3})
  n19 = ScalarParameter() {defaultValue=800, parameterName=NearCameraFadeDis}
  n20 = Multiply(
   A=LinearInterpolate(A=LinearInterpolate(B=@n8, Alpha=@n9) {constA=0.1, constAlpha=1.4257}, B=Multiply(A=@n15, B=@n16), Alpha=Saturate(None=@n17)) {constB=0}
   B=Multiply(
    A=Multiply(A=@n18, B=@n18)
    B=MaterialFunctionCall(Distance=@n19) {materialFunction=&Near_Camera_Fade}.Result
   )
  )
  n21 = TextureSampleParameter2D(UVs=Panner() {speedY=0.3}) {parameterName=DistortionTexture, samplerType=SAMPLERTYPE_Grayscale, texture=&LowResBlurredNoise}
  n22 = TextureCoordinate()
  n23 = TextureSample(UVs=Panner(Coordinate=@n22) {speedX=0.1, speedY=0.1}) {samplerType=SAMPLERTYPE_LinearGrayscale, texture=/Game/Global/Textures/Default/Noise/T_SwirlNoise_N_height}
  n24 = LinearInterpolate(A=@n21.R, B=Multiply(A=@n21.R, B=Multiply(A=@n23.R, B=@n23.R)), Alpha=ScalarParameter() {parameterName=UseSubOpacityMask})
  n25 = DepthFade() {fadeDistanceDefault=200}
  n26 = Constant() {r=1}
  n27 = Multiply(
   A=LinearInterpolate(A=@n20, B=Multiply(A=@n20, B=@n24), Alpha=ScalarParameter() {parameterName=RingUV_Distortion})
   B=ShadingPathSwitch(Default=Multiply(A=@n25, B=@n25), Forward=@n26, Mobile=@n26)
  )
  LinearInterpolate(A=@n27, B=Multiply(A=ParticleColor().A, B=@n27), Alpha=ScalarParameter() {parameterName=UseParticleColor})
  n28 = OneMinus(None=DepthFade(FadeDistance=ScalarParameter() {defaultValue=1200, parameterName=HighLightArea}) {fadeDistanceDefault=1200})
  n29 = Multiply(A=Multiply(A=@n28, B=@n28), B=ScalarParameter() {defaultValue=4.849, parameterName=HighLightIntensity})
  n30 = Constant() {r=1}
  n31 = QualitySwitch(Default=ShadingPathSwitch(Default=@n29, Deferred=@n29, Forward=@n30, Mobile=@n30), Low=@n30, High=@n30, Medium=@n30, Epic=@n29)
  n32 = MaterialFunctionCall(Float2=TextureCoordinate()) {materialFunction=&BreakOutFloat2Components}
  n33 = AppendVector(A=@n32.R, B=Power(Base=@n32.G, Exp=ScalarParameter() {defaultValue=4.411, parameterName=Noise_V_Bias}))
  n34 = DynamicParameter() {paramNames [U_Tile_UVDistortion_A, V_Tile_UVDistortion_A, U_Tile_UVDistortion_B, V_Tile_UVDistortion_B]}
  n35 = Multiply(A=@n33, B=AppendVector(A=@n34.U_Tile_UVDistortion_B, B=@n34.U_Tile_UVDistortion_B))
  n36 = Multiply(
   A=TextureSampleParameter2D(UVs=Panner(Coordinate=@n35) {speedY=0.1}) {parameterName=NoiseTextureB, samplerType=SAMPLERTYPE_Grayscale, texture=&LowResBlurredNoise}.G
   B=ScalarParameter() {defaultValue=2.0265, parameterName=NoiseTextureB_Distorter}
  ) {constB=4}
  n37 = AppendVector(A=@n34.U_Tile_UVDistortion_A, B=@n34.U_Tile_UVDistortion_A)
  n38 = TextureSampleParameter2D(UVs=Panner(Coordinate=Multiply(A=@n33, B=@n37)) {speedY=0.15}) {parameterName=NoiseTextureA, samplerType=SAMPLERTYPE_LinearGrayscale, texture=/Game/Environments/Hexmap/HexClouds/T_HexmapFog_01_Height02}
  n39 = Multiply(A=@n38.R, B=ScalarParameter() {defaultValue=7.6891, parameterName=NoiseTextureA_Distorter})
  n40 = Add(
   A=Divide(A=Add(A=@n36, B=@n39), B=ScalarParameter() {defaultValue=4.7166, parameterName=NoiseTextureDistort_Divider}) {constB=5}
   B=MaterialFunctionCall(Float2=TextureCoordinate()) {materialFunction=&BreakOutFloat2Components}.G
  )
  n41 = Multiply(A=Time()) {constB=-1}
  n42 = CurveAtlasRowParameter(CurveTime=Frac(None=Add(A=@n40, B=@n41))) {atlas=&/CA_AwfulDuck, curve=&/Curve_AwfulDuck_01, parameterName=ColorRange}
  n43 = Constant() {r=-0.3}
  n44 = CurveAtlasRowParameter(CurveTime=MaterialFunctionCall(Float2=TextureCoordinate()) {materialFunction=&BreakOutFloat2Components}.G) {
   atlas=&CA_AwfulDuck
   curve=&/Curve_AwfulDuck_02
   defaultValue=1
   parameterName=ColorGradient
  }
  n45 = LinearInterpolate(A=Desaturation(None=@n42, Fraction=@n43), B=Desaturation(None=@n44, Fraction=@n43)) {constAlpha=0.7}
  n46 = TextureCoordinate()
  n47 = Time()
  n48 = TextureSampleParameter2D() {parameterName=DistortionTexture, samplerType=SAMPLERTYPE_Grayscale, texture=&LowResBlurredNoise}
  n49 = Add(
   A=Add(
    A=MaterialFunctionCall(Float2=@n46) {materialFunction=&BreakOutFloat2Components}.G
    B=Multiply(A=@n47) {constB=-1}
   )
   B=Multiply(A=@n48.R, B=ScalarParameter() {parameterName=RingUV_Distortion})
  )
  n50 = CurveAtlasRowParameter(CurveTime=Frac(None=Multiply(A=@n49) {constB=2})) {
   atlas=&CA_AwfulDuck
   curve=&/Curve_AwfulDuck_03
   defaultValue=2
   parameterName=RingGradient
  }
  n51 = ScalarParameter() {defaultValue=1, parameterName=RingPanningVisibility}
  n52 = Add(A=@n45, B=Multiply(A=@n45, B=Multiply(A=@n50.R, B=@n51)))
  n53 = Multiply(A=Add(A=Multiply(A=@n52, B=@n31), B=@n52), B=ScalarParameter() {defaultValue=1, parameterName=Brightness})
  n54 = ParticleColor()
  n55 = Multiply(A=@n53, B=@n54.RGB)
  n56 = ScalarParameter() {parameterName=UseParticleColor}
  MaterialFunctionCall(Emissive=LinearInterpolate(A=LinearInterpolate(A=@n53, B=@n55, Alpha=@n56), B=@n54.RGB, Alpha=ScalarParameter() {parameterName=Only Particle Color})) {materialFunction=&MF_FX_EmissiveAdjustments}
  Multiply(A=@n22) {constB=0.5}
 ]
}
