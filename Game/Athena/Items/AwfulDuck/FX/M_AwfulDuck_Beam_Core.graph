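# Unlit translucent Niagara ribbon material for the core of the AwfulDuck beam effect.
# Panned noise distorts the cross-beam UV feeding curve-atlas rows that shape the mask and hue-shiftable color, tinted by particle color.
# Opacity combines that mask with end, depth and near-camera fades.
# Dynamic parameters grow the beam along its length and drive a time-animated sine/cosine wave offset, weighted by a curve along U.
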Material /Game/Athena/Items/AwfulDuck/FX/M_AwfulDuck_Beam_Core
graph {
 properties {bUsedWithNiagaraRibbons=true, blendMode=BLEND_Translucent, pixelDepthOffsetMode=PDOM_Legacy, shadingModel=MSM_Unlit, usedShadingModels=Unlit}
 nodes [
  Constant2Vector() {g=1, r=7}
  MaterialFunctionCall() {materialFunction=&DebugTimeSine}
  n1 = VectorParameter() {parameterName=Tile, defaultValue=<rgba 0.2 1.5 0 0>}
  n2 = Multiply(A=TextureCoordinate(), B=AppendVector(A=@n1.R, B=@n1.R))
  n3 = Panner(Coordinate=@n2) {speedX=1, speedY=0.3}
  n4 = LinearInterpolate(
   A=TextureSample(UVs=Panner(Coordinate=@n2) {speedX=0.5, speedY=-0.3}) {texture=&T_Inky_Smoke_Tile}.R
   B=TextureSample(UVs=@n3) {samplerType=SAMPLERTYPE_LinearGrayscale, texture=/Game/Environments/Hexmap/HexClouds/T_HexmapFog_01_Height02}.R
   Alpha=ScalarParameter() {defaultValue=1.7458, parameterName=DistortNoiseBlend}
  ) {constAlpha=0.2}
  n5 = Add(
   A=MaterialFunctionCall(Float2=TextureCoordinate()) {materialFunction=&BreakOutFloat2Components}.G
   B=Divide(A=LinearInterpolate(Alpha=@n4) {constA=-1}) {constB=8}
  )
  n6 = CurveAtlasRowParameter(CurveTime=@n5) {
   atlas=&/CA_AwfulDuck
   curve=&/Curve_AwfulDuck_03
   defaultValue=2
   parameterName=BeamMask
  }
  n7 = CurveAtlasRowParameter(CurveTime=@n6.R) {
   atlas=&CA_AwfulDuck
   curve=&/Curve_AwfulDuck_02
   defaultValue=1
   parameterName=BeamColorCurve
  }
  n8 = MaterialFunctionCall(Hue Shift Percentage=ScalarParameter() {parameterName=Hue_Shift}, Texture=@n7) {materialFunction=&HueShift}
  n9 = ScalarParameter() {defaultValue=1, parameterName=UseColorCurve}
  n10 = ParticleColor()
  MaterialFunctionCall(Emissive=Multiply(A=LinearInterpolate(A=@n5, B=@n8.Result, Alpha=@n9) {constA=1}, B=@n10.RGB)) {materialFunction=&MF_FX_EmissiveAdjustments}
  n11 = LinearInterpolate(Alpha=TextureSample(UVs=@n3) {samplerType=SAMPLERTYPE_LinearGrayscale, texture=/Game/Global/Textures/Default/Noise/T_SwirlNoise_N_height}.R) {constA=-1}
  n12 = MaterialFunctionCall(Float2=TextureCoordinate()) {materialFunction=&BreakOutFloat2Components}
  n13 = CurveAtlasRowParameter(CurveTime=Add(A=Divide(A=@n11) {constB=13}, B=@n12.G)) {
   atlas=&CA_AwfulDuck
   curve=&Curve_AwfulDuck_03
   defaultValue=2
   parameterName=BeamMask
  }
  n14 = OneMinus(None=@n12.R)
  n15 = Multiply(A=Multiply(A=@n13.R, B=@n13.R), B=Saturate(None=Multiply(A=@n14) {constB=15}))
  n16 = DepthFade(Opacity=@n15, FadeDistance=ScalarParameter() {defaultValue=200, parameterName=DepthFade}) {fadeDistanceDefault=200}
  n17 = ShadingPathSwitch(Default=@n16, Deferred=@n16, Forward=@n15, Mobile=@n15)
  n18 = DynamicParameter() {paramNames [WaveTile, WaveIntensity, WaveSpeed, BeamMaskGrowth]}
  n19 = TextureCoordinate()
  n20 = Saturate(None=Subtract(
    A=Multiply(A=@n18.BeamMaskGrowth) {constB=2}
    B=MaterialFunctionCall(Float2=@n19) {materialFunction=&BreakOutFloat2Components}.R
   ))
  n21 = ScalarParameter() {parameterName=NearCameraFade}
  Multiply(
   A=@n10.A
   B=Multiply(
    A=Multiply(A=@n17, B=@n20)
    B=MaterialFunctionCall(Distance=@n21) {materialFunction=&Near_Camera_Fade}.Result
   )
  )
  n22 = CurveAtlasRowParameter(CurveTime=MaterialFunctionCall(Float2=TextureCoordinate()) {materialFunction=&BreakOutFloat2Components}.R) {
   atlas=&CA_AwfulDuck
   curve=&Curve_AwfulDuck_03
   defaultValue=2
   parameterName=BeamMask
  }
  n23 = TextureCoordinate()
  n24 = DynamicParameter() {paramNames [WaveTile, WaveIntensity, WaveSpeed, BeamMaskGrowth]}
  n25 = Add(
   A=Multiply(
    A=MaterialFunctionCall(Float2=@n23) {materialFunction=&BreakOutFloat2Components}.R
    B=@n24.WaveTile
   )
   B=Multiply(A=Time(), B=@n24.WaveSpeed)
  )
  n26 = MaterialFunctionCall(R=Sine(None=@n25), G=Cosine(None=Add(A=@n25) {constB=0}), B=Sine(None=Add(A=@n25) {constB=0.5})) {materialFunction=&AppendMany}
  Multiply(A=@n22.G, B=Multiply(A=Multiply(A=@n26.RGB), B=Multiply(A=@n24.WaveIntensity) {constB=30}) {constB=1000})
  OneMinus(None=@n18.BeamMaskGrowth)
  ScalarParameter() {defaultValue=1.0726, parameterName=Param_1}
 ]
}
