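# Translucent unlit material for the core of the AwfulDuck beam, usable on Niagara ribbons and spline meshes.
# Panning hex-fog noise, U-tiled by a dynamic parameter, wobbles V lookups into a curve-atlas mask that drives a color curve, tinted by particle color for emissive.
# Opacity squares a smoke-distorted mask, then applies a reversible end fade, near-camera fade and particle alpha.
# A shading-path switch skips depth fade on forward and mobile.
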Material /Game/Athena/Items/AwfulDuck/FX/M_AwfulDuck_Beam_CoreV2
graph {
 properties {
  bUsedWithInstancedStaticMeshes=true
  bUsedWithNiagaraRibbons=true
  bUsedWithSplineMeshes=true
  blendMode=BLEND_Translucent
  pixelDepthOffsetMode=PDOM_Legacy
  shadingModel=MSM_Unlit
  usedShadingModels=Unlit
 }
 nodes [
  n1 = TextureCoordinate()
  n2 = DynamicParameter() {paramNames [U_Tile,Param2,Param3,Param4]}
  n3 = Constant() {r=1}
  n4 = Panner(Coordinate=Multiply(A=TextureCoordinate(), B=AppendVector(A=@n2.U_Tile, B=@n3))) {speedX=-0.8, speedY=0.5}
  n5 = LinearInterpolate(Alpha=TextureSample(UVs=@n4) {samplerType=SAMPLERTYPE_LinearGrayscale, texture=/Game/Environments/Hexmap/HexClouds/T_HexmapFog_01_Height02}.R) {constA=-1}
  n6 = CurveAtlasRowParameter(CurveTime=Add(
    A=MaterialFunctionCall(Float2=@n1) {materialFunction=&BreakOutFloat2Components}.G
    B=Divide(A=@n5) {constB=15}
   )) {
   atlas=&/CA_AwfulDuck
   curve=&/Curve_AwfulDuck_03
   defaultValue=2
   parameterName=BeamMask
  }
  n7 = CurveAtlasRowParameter(CurveTime=LinearInterpolate(A=@n6.R, B=@n6.R)) {
   atlas=&CA_AwfulDuck
   curve=&/Curve_AwfulDuck_02
   defaultValue=1
   parameterName=BeamColorCurve
  }
  n8 = ParticleColor()
  MaterialFunctionCall(Emissive=Multiply(A=@n7, B=@n8.RGB)) {materialFunction=&MF_FX_EmissiveAdjustments}
  n9 = LinearInterpolate(Alpha=TextureSample(UVs=@n4) {texture=&T_Inky_Smoke_Tile}.R) {constA=-1}
  n10 = MaterialFunctionCall(Float2=TextureCoordinate()) {materialFunction=&BreakOutFloat2Components}
  n11 = CurveAtlasRowParameter(CurveTime=Add(A=Divide(A=@n9) {constB=15}, B=@n10.G)) {
   atlas=&CA_AwfulDuck
   curve=&Curve_AwfulDuck_03
   defaultValue=2
   parameterName=BeamMask
  }
  n12 = OneMinus(None=@n10.R)
  n13 = ScalarParameter() {defaultValue=1, parameterName=ReverseEndMask}
  n14 = Multiply(
   A=Multiply(A=@n11.G, B=@n11.G)
   B=Multiply(A=LinearInterpolate(A=@n12, B=@n10.R, Alpha=@n13), B=ScalarParameter() {defaultValue=15, parameterName=EndFadeoutFalloff})
  )
  n15 = DepthFade(Opacity=@n14, FadeDistance=ScalarParameter() {defaultValue=800, parameterName=DepthFadeDistance}) {fadeDistanceDefault=800}
  n16 = ScalarParameter() {defaultValue=2000, parameterName=NearCameraFadeDis}
  Multiply(
   A=@n8.A
   B=Multiply(
    A=ShadingPathSwitch(Default=@n15, Deferred=@n15, Forward=@n14, Mobile=@n14)
    B=MaterialFunctionCall(Distance=@n16) {materialFunction=&Near_Camera_Fade}.Result
   )
  )
 ]
}
