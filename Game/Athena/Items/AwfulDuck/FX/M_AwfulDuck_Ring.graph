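# Translucent unlit material for Niagara mesh particles, drawing the AwfulDuck item's ring effect.
# SplineThicken sets width from a parameter, the Thickness dynamic parameter and per-particle random variation.
# Color blends two parameters along U with a sine; opacity comes from a squared curve-atlas lookup, contrast-adjusted by view-to-tangent angle and depth-faded on the default shading path.
# Optional particle alpha and scrolling smoke-texture erosion, driven by the Erode dynamic parameter, further shape opacity.
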Material /Game/Athena/Items/AwfulDuck/FX/M_AwfulDuck_Ring
graph {
 properties {
  bEnableResponsiveAA=true
  bUsedWithNiagaraMeshParticles=true
  blendMode=BLEND_Translucent
  pixelDepthOffsetMode=PDOM_Legacy
  shadingModel=MSM_Unlit
  twoSided=true
  usedShadingModels=Unlit
 }
 nodes [
  n1 = MaterialFunctionCall(Float2=TextureCoordinate()) {materialFunction=&BreakOutFloat2Components}
  n2 = LinearInterpolate(
   A=VectorParameter() {parameterName=ColorLow, defaultValue=<rgba 0.458 0 8 1>}.RGB
   B=VectorParameter() {parameterName=ColorHigh, defaultValue=<rgba 0.6445 1 0.4893 1>}.RGB
   Alpha=Sine(None=Multiply(A=@n1.R) {constB=0.5})
  )
  n3 = Multiply(A=ParticleColor().RGB, B=@n2)
  LinearInterpolate(A=@n2, B=Multiply(A=@n3, B=ParticleColor().RGB), Alpha=ScalarParameter() {parameterName=UseParticleColor})
  n4 = DynamicParameter() {defaultValue=<rgba 1 0.75 1 1>, paramNames [Thickness,Erode,Param3,Param4]}
  n5 = Multiply(
   A=Multiply(A=ScalarParameter() {defaultValue=50, parameterName=SplineWidth}, B=@n4.Thickness)
   B=LinearInterpolate(Alpha=ParticleRandom()) {constA=0.7, constB=1.3}
  )
  MaterialFunctionCall(WidthBase=@n5, WidthTip=@n5) {materialFunction=&SplineThicken}
  MaterialFunctionCall(Emissive=@n3) {materialFunction=&MF_FX_EmissiveAdjustments}
  n6 = Abs(None=DotProduct(A=CameraVectorWS(), B=VertexTangentWS()))
  Multiply(A=@n6, B=@n6)
  n7 = CurveAtlasRowParameter(CurveTime=@n1.R) {atlas=&/CA_AwfulDuck, curve=&/Curve_AwfulDuck_03, defaultValue=2}
  n8 = Multiply(A=@n7.G, B=@n7.G)
  n9 = LinearInterpolate(
   A=MaterialFunctionCall(In=@n8, Contrast=Constant() {r=0.3}) {materialFunction=&CheapContrast}.Result
   B=@n8
   Alpha=Saturate(None=Multiply(A=@n6) {constB=2})
  )
  n10 = ShadingPathSwitch(Default=DepthFade(Opacity=@n9, FadeDistance=ScalarParameter() {defaultValue=50, parameterName=DepthFadeDistance}), Forward=@n9, Mobile=@n9)
  n11 = ParticleColor()
  n12 = TextureSample(UVs=Panner(Coordinate=TextureCoordinate() {vTiling=5}) {speedY=1}) {texture=&T_Inky_Smoke_Tile}
  Multiply(
   A=LinearInterpolate(A=@n10, B=Multiply(A=@n11.A, B=@n10), Alpha=ScalarParameter() {parameterName=UseParticleColor})
   B=LinearInterpolate(
    B=MaterialFunctionCall!(Dynamic Param=@n4.Erode, Alpha=@n12.R) {materialFunction=/Game/Effects/Fort_Effects/Materials/_Functions/Erode_Alpha}.Result
    Alpha=ScalarParameter() {parameterName=Erode viaTexture}
   ) {constA=1}
  )
 ]
}
