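# Niagara system for the AwfulDuck item that, despite its DirtLiftup name, contains only a WarningBeam emitter drawing a translucent cyan ribbon beam.
# A single burst of particles forms the beam from the simulation position to the absolute point in User.BeamTarget.
# Width tapers along ribbon link order, while update curves ramp width and alpha up from zero.
# The User.DirtInitialVeclocityDirection parameter is exposed but unused by this emitter.
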NiagaraSystem /Game/Athena/Items/AwfulDuck/FX/NS_AwfulDuck_Local_DirtLiftup
dependencies {
 usedDataInterfaces [&NiagaraDataInterfaceCurve]
 usedDynamicInputs [&Multiply_VectorByFloat]
 usedModules [&SystemState]
}
emitters {WarningBeam {
  emitterUpdateScript [
   &EmitterState(Life Cycle Mode=::NewEnumerator1, Loop Behavior=::NewEnumerator1, Loop Duration Mode=::NewEnumerator1, UseLoopDelay=false)
   /Niagara/Modules/Beams/BeamEmitterSetup(
    Absolute Beam End=true
    Beam End:Vector3f=$User.BeamTarget
    Beam Start:Vector3f=&Add_Vector(A=/Niagara/DynamicInputs/Helpers/SimulationPosition(), B=<0 0 0>)
   )
   &SpawnBurst_Instantaneous(Spawn Count=10, Use Spawn Probability=false)
  ]
  particleSpawnScript [
   &InitializeParticle(Color=<rgba 0.1906 1.7962 3 0.2>, Position Mode=::NewEnumerator0, Write Material Random=true)
   /Niagara/Modules/Beams/SpawnBeam()
   /Niagara/Modules/Beams/BeamWidth(Beam Width:NiagaraFloat=&FloatFromCurve(
     FloatCurve=DataInterface() {propertyValues {Curve {keys [
         {interpMode=RCIM_Cubic, tangentMode=RCTM_User, time=0, value=1.175}
         {arriveTangent=-0.875, interpMode=RCIM_Cubic, leaveTangent=-0.875, time=0.3, value=0.9}
         {arriveTangent=-0.8137, interpMode=RCIM_Cubic, leaveTangent=-0.8137, tangentMode=RCTM_User, time=1, value=0.3}
        ]}}}
     CurveIndex=$Particles.RibbonLinkOrder
     Scale Curve=250
    ))
  ]
  particleUpdateScript [
   &ParticleState(Kill Particles When Lifetime Has Elapsed=false, Let Infinitely Lived Particles Die When Emitter Deactivates=true)
   /Niagara/Modules/Beams/UpdateBeam()
   &SolveForcesAndVelocity(Clamp Velocity=false, Limit Acceleration=false)
   /Niagara/Modules/Beams/ScaleBeamWidth(Beam Width Scale:NiagaraFloat=&FloatFromCurve(FloatCurve=DataInterface() {propertyValues {Curve {keys [<at 0 0>,<at 0.2947 0.9801>,<at 1 1>]}}}))
   &ScaleColor(
    Scale Alpha=&FloatFromCurve(FloatCurve=DataInterface() {propertyValues {Curve {keys [<at 0 0>,<at 0.3 1>,<at 1 1>]}}})
    ScaleA=true
    ScaleRGB=true
   )
  ]
  rendererClasses [&NiagaraRibbonRendererProperties]
  renderers [{rendererClass=&NiagaraRibbonRendererProperties}]
 }}
summary {
 systemName=NS_AwfulDuck_Local_DirtLiftup
 emitters [{emitterName=WarningBeam, rendererClasses [&NiagaraRibbonRendererProperties]}]
 userVariables [
  User.BeamTarget {type=Vector3f, description="", defaultValue {struct=Vector3f, value=<8700 0 1500>}}
  User.DirtInitialVeclocityDirection {type=Vector3f, description="", defaultValue {struct=Vector3f, value=<0 0 1>}}
 ]
}
