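# Niagara effect for the AwfulDuck item that pairs a colored light with drifting debris and sparks, both distance-culled.
# The Light emitter bursts one teal light particle, offset above the system.
# A GPU emitter continuously spawns tinted chunks in a cylinder, pushed along the user-set dirt direction vector with curl noise.
# Those particles feed one mesh and two sprite renderers, with curve-driven size fades, spin and randomized spark color and size.
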NiagaraSystem /Game/Athena/Items/AwfulDuck/FX/NS_AwfulDuck_Local_Light
dependencies {
 usedDataInterfaces [&NiagaraDataInterfaceCurve, &NiagaraDataInterfaceVectorCurve]
 usedDynamicInputs []
 usedModules [&SystemState]
}
emitters {
 Light {
  emitterUpdateScript [
   &EmitterState(
    Enable Distance Culling=true
    Life Cycle Mode=::NewEnumerator1
    Loop Behavior=::NewEnumerator1
    Loop Duration Mode=::NewEnumerator1
    Max Distance Response=::SleepAndClearParticles
    MaxDistance=6000
    Module.bUseMaxDistance=true
    Module.bUseMinDistance=true
    Reset Age On Awaken=true
    Scalability Mode=::NewEnumerator1
    UseLoopDelay=false
   )
   &SpawnBurst_Instantaneous(Spawn Count=1, Use Spawn Probability=false)
  ]
  particleSpawnScript [
   &InitializeParticle(Color=<rgba 0.9034 14 12.3239 1>, UsePositionOffset=false, Write Material Random=true)
   &SystemLocation(MaskSpawnGroup=false, Offset=<0 0 500>)
  ]
  particleUpdateScript [&ParticleState(Kill Particles When Lifetime Has Elapsed=false, Let Infinitely Lived Particles Die When Emitter Deactivates=true)]
  rendererClasses [&NiagaraLightRendererProperties]
  renderers [{rendererClass=&NiagaraLightRendererProperties, rendererIndex=0}]
 }
 NE_FN_DirtChunks {
  simTarget=GPUComputeSim
  emitterUpdateScript [
   &EmitterState(Enable Distance Culling=true, MaxDistance=1500, Module.bUseMaxDistance=true, Module.bUseMinDistance=true, Scalability Mode=::NewEnumerator1)
   &SpawnRate(SpawnRate=40, Use Spawn Probability=false)
   &SpawnRate(SpawnGroup=1, SpawnRate=40, Use Spawn Probability=false) {module=SpawnRate001}
  ]
  particleSpawnScript [
   &InitializeParticle(
    Color=&RandomRangeLinearColor(Minimum=<rgba 0.4844 0.4844 0.4844 1>, Maximum=<rgba 0.8683 10 8.2982 1>)
    Lifetime Max=5
    Lifetime Min=1.5
    Lifetime Mode=::NewEnumerator1
    Mesh Scale=&Multiply_VectorByFloat(
     Vector=&UniformRangedVector(Minimum=<0.2 0.4 0.4>, Maximum=<0.6 1 1.1>)
     Float=&UniformRangedFloat(Minimum=0.02, Maximum=0.05)
    )
    Mesh Scale Mode=::NewEnumerator4
    Sprite Size Max=<5 5>
    Sprite Size Min=<1 1>
    Sprite Size Mode=::NewEnumerator2
    UsePositionOffset=false
    Write Material Random=true
   )
   &InitialMeshOrientation(Orientation Vector=&RandomVector(Vector Scale=1), Use Orientation Vector=true, Use Rotation Vector=false)
   &CylinderLocation(Cylinder Radius=800, MaskSpawnGroup=false, Override Seed=false, UseOwnerScale=true)
   &AddVelocityInCone(
    Cone Angle=150
    Cone Axis=$User.DirtInitialVeclocityDirection
    Module.Use Velocity Falloff On Cone Axis=false
    Velocity Strength=&RandomRangeFloat(Minimum=5, Maximum=50)
   )
   &ApplyInitialForces()
   &DynamicMaterialParameters(Module.Param0WriteEnabled=true, Module.Param1WriteEnabled=false, Module.Param2WriteEnabled=false, Module.Param3WriteEnabled=false)
   /SetVariables(Particles.SpawnGroup=&MakeIntFromFloat(Float=&FloatFromCurve(
      FloatCurve=DataInterface() {propertyValues {Curve {keys [
          {tangentMode=RCTM_User, time=0, value=0}
          {arriveTangent=-0.0405, interpMode=RCIM_Cubic, leaveTangent=-0.0405, tangentMode=RCTM_User, time=0.9, value=0}
          {arriveTangent=5.2322, interpMode=RCIM_Cubic, leaveTangent=5.2322, tangentMode=RCTM_User, time=1, value=1}
         ]}}}
      CurveIndex=&RandomRangeFloat(Minimum=-1, Maximum=2)
     ))) {isSetParameters=true}
   /SetVariables(
    Particles.SparkColor:LinearColor=&RandomRangeLinearColor(Minimum=<rgba 0.5776 1 0.1354 0.1>, Maximum=<rgba 0.0208 1 1 1>)
    Particles.SparkSize:Vector2f=<50 50>
   ) {isSetParameters=true}
   /Niagara/Modules/Update/SubUV/SubUVAnimation(SubUV Animation Mode=::NewEnumerator1)
  ]
  particleUpdateScript [
   &ParticleState()
   &ScaleColor(
    Scale RGB=&VectorFromCurve(
     VectorCurve=DataInterface() {propertyValues {
       XCurve {keys [
         {arriveTangent=-10.0968, interpMode=RCIM_Cubic, time=0, value=2}
         {arriveTangent=-5, interpMode=RCIM_Cubic, leaveTangent=-5, time=0.2, value=0.6}
         <at 0.4 0>
        ]}
       YCurve {keys [
         {arriveTangent=-16.8109, interpMode=RCIM_Cubic, time=0, value=0.6}
         {arriveTangent=-1.6216, interpMode=RCIM_Cubic, leaveTangent=-1.6216, time=0.2, value=0.12}
         <at 0.37 0>
        ]}
       ZCurve {keys [
         {arriveTangent=-3.9061, interpMode=RCIM_Cubic, leaveTangent=-1, tangentMode=RCTM_User, time=0, value=0.25}
         {arriveTangent=-0.6757, interpMode=RCIM_Cubic, leaveTangent=-0.6757, time=0.2, value=0.04}
         <at 0.37 0>
        ]}
      }}
     Scale Curve=&VectorFromFloat(Value=1)
    )
    ScaleA=false
    ScaleRGB=true
   ) {enabled=false}
   &ScaleColor(ScaleA=false, ScaleRGB=true) {module=ScaleColor001}
   &AccelerationForce(Acceleration=&Multiply_VectorByFloat(Vector=$User.DirtInitialVeclocityDirection, Float=180))
   &CurlNoiseForce(Noise Strength=100, Pan Noise=false)
   &SolveForcesAndVelocity(Clamp Velocity=false, Limit Acceleration=false)
   &SolveRotationalForcesAndVelocity()
   &ScaleMeshSize(Scale Factor=&VectorFromFloat(Value=&FloatFromCurve(FloatCurve=DataInterface() {propertyValues {Curve {keys [
          {interpMode=RCIM_Cubic, tangentMode=RCTM_User, time=0, value=0}
          {interpMode=RCIM_Cubic, tangentMode=RCTM_User, time=0.2, value=1.05}
          {arriveTangent=-1.3125, interpMode=RCIM_Cubic, leaveTangent=-1.3125, time=0.8, value=0.95}
          {tangentMode=RCTM_User, time=1, value=0}
         ]}}})))
   &ScaleSpriteSize(Scale Factor=&Vector2DFromFloat(Value=&FloatFromCurve(FloatCurve=DataInterface() {propertyValues {Curve {keys [
          {interpMode=RCIM_Cubic, tangentMode=RCTM_User, time=0, value=0}
          {arriveTangent=1.1875, interpMode=RCIM_Cubic, leaveTangent=1.1875, time=0.1, value=1.05}
          {arriveTangent=-1.1667, interpMode=RCIM_Cubic, leaveTangent=-1.1667, time=0.8, value=0.95}
          {tangentMode=RCTM_User, time=1, value=0}
         ]}}})))
   /SetVariables(Particles.SparkSize:Vector2f=&Vector2DFromFloat(Value=&FloatFromCurve(
      FloatCurve=DataInterface() {propertyValues {Curve {keys [
          {interpMode=RCIM_Cubic, tangentMode=RCTM_User, time=0, value=0}
          {interpMode=RCIM_Cubic, tangentMode=RCTM_User, time=0.9286, value=1.0635}
          <at 0.9972 0 RCIM_Cubic>
         ]}}}
      Scale Curve=&RandomRangeFloat(Minimum=1, Maximum=20)
     ))) {isSetParameters=true}
   &SpriteRotationRate(Rotation Rate=&Lerp_Float(A=-270, B=270, Alpha=$Particles.Initial.MaterialRandom))
   &SpriteRotationRate(Rotation Rate=5) {module=SpriteRotationRate001}
   &UpdateMeshOrientation(Rotation Rate=0.5, Rotation Vector=<1 1 1>)
  ]
  rendererClasses [&NiagaraMeshRendererProperties, &NiagaraSpriteRendererProperties]
  renderers [
   {rendererClass=&NiagaraMeshRendererProperties, rendererIndex=0}
   {rendererClass=&NiagaraSpriteRendererProperties, rendererIndex=1}
   {rendererClass=&NiagaraSpriteRendererProperties, rendererIndex=2}
  ]
 }
}
summary {
 systemName=NS_AwfulDuck_Local_Light
 emitters [
  {
   emitterName=NE_FN_DirtChunks
   simTarget=GPUComputeSim
   rendererClasses [&NiagaraMeshRendererProperties, &NiagaraSpriteRendererProperties]
  }
  {emitterName=Light, rendererClasses [&NiagaraLightRendererProperties]}
 ]
 userVariables [User.DirtInitialVeclocityDirection {type=Vector3f, description="", defaultValue {struct=Vector3f, value=<0 0 1>}}]
}
