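# Niagara system for the AwfulDuck item's telegraph warm-up pulse, a short sprite effect on a skeletal mesh.
# One emitter bursts 17 camera-offset sprites, each assigned a spawn group that serves as its bone/socket index.
# Skeletal mesh location runs in particle update, keeping sprites on the user-supplied mesh.
# Over each particle's life, curves fade alpha in and out, shift color from blue toward purple and drive dynamic material parameters.
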NiagaraSystem /Game/Athena/Items/AwfulDuck/FX/NS_AwfulDuck_Telegraph_WarmUpPulse
dependencies {
 usedDataInterfaces [&NiagaraDataInterfaceColorCurve, &NiagaraDataInterfaceCurve]
 usedDynamicInputs []
 usedModules [&SystemState]
}
emitters {Pulse_PDecal {
  emitterUpdateScript [
   &EmitterState(Life Cycle Mode=::NewEnumerator1, Loop Behavior=::NewEnumerator1, Loop Duration=1, UseLoopDelay=false)
   &SpawnBurst_Instantaneous(Spawn Count=17, Use Spawn Probability=false)
  ]
  particleSpawnScript [
   &InitializeParticle(
    Color=<rgba 0.5728 0.6796 1 1>
    Lifetime=1.5
    Sprite Size Mode=::NewEnumerator3
    Uniform Sprite Size=50
    UsePositionOffset=false
    Write Material Random=true
   )
   &CameraOffset(Camera Offset Amount=50)
   &FN_SetSpawnGroup(Number of Spawn Groups=17)
  ]
  particleUpdateScript [
   &ParticleState()
   &ScaleColor(
    Linear Color Curve=DataInterface() {propertyValues {
      AlphaCurve {keys [<at 0 0 RCIM_Cubic>, {arriveTangent=-1.373, interpMode=RCIM_Cubic, leaveTangent=-1.373, tangentMode=RCTM_User, time=0.3, value=1}, <at 1 0>]}
      BlueCurve {keys [<at 0 1>,<at 0.6831 1>]}
      GreenCurve {keys [<at 0 0.3636>,<at 0.6831 0>]}
      RedCurve {keys [<at 0 0>,<at 0.6831 0.4834>]}
     }}
    Scale Mode=::NewEnumerator2
   )
   &DynamicMaterialParameters(
    Index 0 Param 1=&FloatFromCurve(FloatCurve=DataInterface() {propertyValues {Curve {keys [
         <at 0 -0.5 RCIM_Cubic>
         {arriveTangent=21.4883, interpMode=RCIM_Cubic, leaveTangent=21.4883, tangentMode=RCTM_User, time=0.07, value=0.27}
         {arriveTangent=5.7766, interpMode=RCIM_Cubic, leaveTangent=5.7766, tangentMode=RCTM_User, time=0.15, value=1}
         {arriveTangent=0.5047, interpMode=RCIM_Cubic, leaveTangent=0.5047, tangentMode=RCTM_User, time=1, value=5}
        ]}}})
    Index 0 Param 2=1
    Index 0 Param 3=&FloatFromCurve(FloatCurve=DataInterface() {propertyValues {Curve {keys [{arriveTangent=1.3198, interpMode=RCIM_Cubic, leaveTangent=1.3198, tangentMode=RCTM_User, time=0, value=-0.25}, <at 1 1>]}}})
    Index 0 Param 4=1
    Module.Param0WriteEnabled=true
    Module.Param1WriteEnabled=true
    Module.Param2WriteEnabled=true
    Module.Param3WriteEnabled=true
   )
   &SkeletalMeshLocation(
    Bone / Socket Index=$Particles.SpawnGroup
    Bone Sampling Mode=::NewEnumerator4
    Inherit Velocity=false
    MaskSpawnGroup=false
    Skeletal Mesh=$User.SkeletalMesh
   )
  ]
  rendererClasses [&NiagaraSpriteRendererProperties]
  renderers [{rendererClass=&NiagaraSpriteRendererProperties}]
 }}
summary {
 systemName=NS_AwfulDuck_Telegraph_WarmUpPulse
 emitters [{emitterName=Pulse_PDecal, rendererClasses [&NiagaraSpriteRendererProperties]}]
 userVariables [User.SkeletalMesh {
   type=NiagaraDataInterfaceSkeletalMesh
   defaultValue {
    struct=NiagaraExt_VariableValue_DataInterface
    value {dataInterface=&self:NiagaraDataInterfaceSkeletalMesh_0}
   }
  }]
}
